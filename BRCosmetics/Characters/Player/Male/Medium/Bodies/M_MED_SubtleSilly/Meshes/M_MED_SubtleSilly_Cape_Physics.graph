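# Physics asset for the cape of the SubtleSilly medium male player body, built entirely from capsules.
# A small base_cape_01 body anchors a chain of six dyn_cape segments, with twist locked, a nearly rigid first joint and limited swing below it.
# Unconstrained bodies on the back, front and side skirt, left sleeve, left arm and sword appear to serve as colliders for the cape, covering only the left side and back.
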PhysicsAsset /BRCosmetics/Characters/Player/Male/Medium/Bodies/M_MED_SubtleSilly/Meshes/M_MED_SubtleSilly_Cape_Physics
bodies {
 base_cape_01 [{shapeType=Capsule, length=0.505, radius=0.505, shapeName=base_cape_01_capsule}]
 dyn_cape_01 [{shapeType=Capsule, length=40, radius=9.4, shapeName=dyn_cape_01_capsule, center=<2 0 11>, rotation=<rot 0 0 90>}]
 dyn_cape_02 [{shapeType=Capsule, length=40, radius=9.4, shapeName=dyn_cape_02_capsule, center=<2 0 11.5>, rotation=<rot 0 0 90>}]
 dyn_cape_03 [
  {shapeType=Capsule, length=40, radius=9.4, shapeName=dyn_cape_03_capsule, center=<2 0 7.5>, rotation=<rot 0 0 90>}
  {
   shapeType=Capsule
   length=12.4628
   radius=3.1332
   shapeName=dyn_cape_03_capsule
   center=<7.2868 -13.8047 -2.3409>
   rotation=<rot 76.7822 -8.0967 9.136>
  }
 ]
 dyn_cape_04 [{shapeType=Capsule, length=40.4025, radius=8.9975, shapeName=dyn_cape_04_capsule, center=<2 0 3>, rotation=<rot 0 0 90>}]
 dyn_cape_05 [{shapeType=Capsule, length=40, radius=9.4, shapeName=dyn_cape_05_capsule, center=<2 0 0.7>, rotation=<rot 0 0 90>}]
 dyn_cape_06 [
  {shapeType=Capsule, length=40, radius=9.4, shapeName=dyn_cape_06_capsule, center=<5 0 0>, rotation=<rot 0 0 90>}
  {shapeType=Capsule, length=38.3556, radius=11.0444, shapeName=dyn_cape_06_capsule, center=<28.7673 0 -1.3417>, rotation=<rot 0 0 90>}
 ]
 dyn_skirt_back_1 [{shapeType=Capsule, length=19.4047, radius=5.5953, shapeName=dyn_skirt_back_1_capsule}]
 dyn_skirt_back_3 [{
   shapeType=Capsule
   length=48.053
   radius=9.9167
   shapeName=dyn_skirt_back_3_capsule
   center=<7.3218 4.7448 2.2315>
   rotation=<rot 90 -8.2203 0>
  }]
 dyn_skirt_back_side_l_1 [{shapeType=Capsule, length=19.4047, radius=5.5953, shapeName=dyn_skirt_back_side_l_1_capsule}]
 dyn_skirt_back_side_l_3 [{
   shapeType=Capsule
   length=48.1238
   radius=9.846
   shapeName=dyn_skirt_back_side_l_3_capsule
   center=<4.4958 6.5484 0.358>
   rotation=<rot 87.0372 -0.0001 1.415>
  }]
 dyn_skirt_front_l_1 [{shapeType=Capsule, length=19.4047, radius=5.5953, shapeName=dyn_skirt_front_l_1_capsule}]
 dyn_skirt_front_l_2 [{
   shapeType=Capsule
   length=45.1279
   radius=9.5935
   shapeName=dyn_skirt_front_l_2_capsule
   center=<18.3961 4.6734 -0.3105>
   rotation=<rot 87.5927 -0.0036 -0.0032>
  }]
 dyn_skirt_side_l_1 [{shapeType=Capsule, length=13.7525, radius=5.5953, shapeName=dyn_skirt_side_l_1_capsule, center=<0 6.1138 0>}]
 dyn_skirt_side_l_2 [{
   shapeType=Capsule
   length=44.8038
   radius=10.3971
   shapeName=dyn_skirt_side_l_2_capsule
   center=<19.0551 6.9112 -2.1007>
   rotation=<rot 79.6426 0 5.4213>
  }]
 dyn_sleeve_lf [{
   shapeType=Capsule
   length=11.2589
   radius=4.0657
   shapeName=dyn_sleeve_lf_capsule
   center=<10.3462 -3.4224 -4.6599>
   rotation=<rot 68.2549 0 0>
  }]
 dyn_sword [
  {
   shapeType=Capsule
   length=52.5526
   radius=3.051
   shapeName=dyn_sword_capsule
   center=<7.8105 6.414 -0.4099>
   rotation=<rot -0.0871 -3.9473 -88.3667>
  }
  {
   shapeType=Capsule
   length=12.3021
   radius=2.098
   shapeName=dyn_sword_capsule
   center=<8.1031 30.5433 -1.6386>
   rotation=<rot 88.4518 0.0001 0.0001>
  }
 ]
 lowerarm_l [{shapeType=Capsule, length=21.4468, radius=7.5532, shapeName=lowerarm_l_capsule, center=<10.1837 0.4886 0>, rotation=<rot 90 0 0>}]
 upperarm_l [{
   shapeType=Capsule
   length=23.1857
   radius=8.9926
   shapeName=upperarm_l_capsule
   center=<10.3058 0.1623 -1.1312>
   rotation=<rot 76.6287 88.2619 91.7497>
  }]
}
constraints [
 {bone1Name=dyn_cape_01, bone2Name=base_cape_01, swing1LimitDegrees=0, swing2LimitDegrees=5, twistLimitDegrees=0, twistMotion=Locked}
 {bone1Name=dyn_cape_02, bone2Name=dyn_cape_01, swing1LimitDegrees=20, swing2LimitDegrees=20, twistLimitDegrees=0, twistMotion=Locked}
 {bone1Name=dyn_cape_03, bone2Name=dyn_cape_02, swing1LimitDegrees=20, swing2LimitDegrees=20, twistLimitDegrees=0, twistMotion=Locked}
 {bone1Name=dyn_cape_04, bone2Name=dyn_cape_03, swing1LimitDegrees=20, swing2LimitDegrees=10, twistLimitDegrees=0, twistMotion=Locked}
 {bone1Name=dyn_cape_05, bone2Name=dyn_cape_04, swing1LimitDegrees=20, swing2LimitDegrees=10, twistLimitDegrees=0, twistMotion=Locked}
 {bone1Name=dyn_cape_06, bone2Name=dyn_cape_05, swing1LimitDegrees=20, swing2LimitDegrees=10, twistLimitDegrees=0, twistMotion=Locked}
]
