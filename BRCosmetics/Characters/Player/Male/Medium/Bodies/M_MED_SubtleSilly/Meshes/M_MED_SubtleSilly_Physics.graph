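# Physics asset for the M_MED_SubtleSilly male player body cosmetic, simulating skirt, loincloth and strap bone chains.
# Capsule bodies on dyn_ bones form limited-swing chains, but only loincloth and back skirt chains are constrained to base_ anchors.
# Thigh, calf and oversized floor capsules supply collision shapes.
# Locked constraints join the three back-skirt tips, while a second, free constraint duplicates the limited front_l_3 to front_l_2 link.
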PhysicsAsset /BRCosmetics/Characters/Player/Male/Medium/Bodies/M_MED_SubtleSilly/Meshes/M_MED_SubtleSilly_Physics
bodies {
 base_loincloth_01 [{shapeType=Capsule, length=0.505, radius=0.505, shapeName=base_loincloth_01_capsule, rotation=<rot 0 -90 0>}]
 base_skirt_back_0 [{
   shapeType=Capsule
   length=46.3023
   radius=15.4637
   shapeName=base_skirt_back_0_capsule
   center=<42.651 6.2672 -2.2093>
   rotation=<rot 0 -91.3458 -84.5281>
  }]
 base_skirt_back_side_l_0 [{
   shapeType=Capsule
   length=44.1809
   radius=13.3668
   shapeName=base_skirt_back_side_l_0_capsule
   center=<41.3289 6.5084 -3.8868>
   rotation=<rot 0 -91.8033 -86.2821>
  }]
 base_skirt_back_side_r_0 [{
   shapeType=Capsule
   length=49.816
   radius=11.3078
   shapeName=base_skirt_back_side_r_0_capsule
   center=<46.1173 5.6506 -0.2087>
   rotation=<rot 0 -93.8846 -86.8521>
  }]
 base_skirt_front_l_1 [{shapeType=Capsule, length=0.505, radius=0.505, shapeName=base_skirt_front_l_1_capsule, rotation=<rot 0 -90 0>}]
 base_skirt_front_r_1 [{shapeType=Capsule, length=0.505, radius=0.505, shapeName=base_skirt_front_r_1_capsule, rotation=<rot 0 -90 0>}]
 base_skirt_side_l_1 [{shapeType=Capsule, length=0.505, radius=0.505, shapeName=base_skirt_side_l_1_capsule, rotation=<rot 0 -90 0>}]
 base_skirt_side_r_1 [{shapeType=Capsule, length=0.505, radius=0.505, shapeName=base_skirt_side_r_1_capsule, rotation=<rot 0 -90 0>}]
 base_strap_lf_01 [{shapeType=Capsule, length=0.505, radius=0.505, shapeName=base_strap_lf_01_capsule, rotation=<rot 0 -90 0>}]
 base_strap_rt_01 [{shapeType=Capsule, length=0.505, radius=0.505, shapeName=base_strap_rt_01_capsule, rotation=<rot 0 -90 0>}]
 calf_l [{shapeType=Capsule, length=51.5266, radius=6.9734, shapeName=calf_l_capsule, center=<-25 -1.2 -0.6>, rotation=<rot 90 0 0>}]
 calf_r [{shapeType=Capsule, length=51.5266, radius=6.9734, shapeName=calf_l_capsule, center=<25 1.2 0.6>, rotation=<rot 90 0 0>}]
 dyn_loincloth_01 [{shapeType=Capsule, length=10.7508, radius=2.7492, shapeName=dyn_loincloth_01_capsule, center=<5 0 0>, rotation=<rot 0 0 90>}]
 dyn_loincloth_02 [{shapeType=Capsule, length=10.2508, radius=3.2492, shapeName=dyn_loincloth_02_capsule, center=<5 0 0>, rotation=<rot 0 0 90>}]
 dyn_loincloth_03 [{shapeType=Capsule, length=9.7508, radius=3.7492, shapeName=dyn_loincloth_03_capsule, center=<5 0 0>, rotation=<rot 0 0 90>}]
 dyn_loincloth_04 [{shapeType=Capsule, length=9.2508, radius=4.2492, shapeName=dyn_loincloth_04_capsule, center=<5 0 0>, rotation=<rot 0 0 90>}]
 dyn_skirt_back_0 [{shapeType=Capsule, length=20, radius=10, shapeName=dyn_skirt_back_0_capsule, center=<10 -7 0>}]
 dyn_skirt_back_1 [{shapeType=Capsule, length=32.5053, radius=7, shapeName=dyn_skirt_back_1_capsule, center=<2 -5 5>}]
 dyn_skirt_back_2 [{shapeType=Capsule, length=32.5053, radius=7, shapeName=dyn_skirt_back_2_capsule, center=<2 -4 5>}]
 dyn_skirt_back_3 [{shapeType=Capsule, length=32.5053, radius=7, shapeName=dyn_skirt_back_3_capsule, center=<2 -3.1 5>}]
 dyn_skirt_back_4 [{shapeType=Capsule, length=32.5053, radius=7, shapeName=dyn_skirt_back_4_capsule, center=<5 -1 5>}]
 dyn_skirt_back_side_l_0 [{shapeType=Capsule, length=20, radius=10, shapeName=dyn_skirt_back_side_l_0_capsule, center=<10 -8.2 0>}]
 dyn_skirt_back_side_l_1 [{shapeType=Capsule, length=32.5053, radius=7, shapeName=dyn_skirt_back_side_l_1_capsule, center=<2 -5.9 0>}]
 dyn_skirt_back_side_l_2 [{shapeType=Capsule, length=32.5053, radius=7, shapeName=dyn_skirt_back_side_l_2_capsule, center=<2 -4.5 0>}]
 dyn_skirt_back_side_l_3 [{shapeType=Capsule, length=32.5053, radius=7, shapeName=dyn_skirt_back_side_l_3_capsule, center=<2 -3.1 0>}]
 dyn_skirt_back_side_l_4 [{shapeType=Capsule, length=32.5053, radius=7, shapeName=dyn_skirt_back_side_l_4_capsule, center=<5 -1.2 0>}]
 dyn_skirt_back_side_r_0 [{shapeType=Capsule, length=20, radius=10, shapeName=dyn_skirt_back_side_r_0_capsule, center=<10 -8.2 0>}]
 dyn_skirt_back_side_r_1 [{shapeType=Capsule, length=32.5053, radius=7, shapeName=dyn_skirt_back_side_r_1_capsule, center=<2 -5.9 0>}]
 dyn_skirt_back_side_r_2 [{shapeType=Capsule, length=32.5053, radius=7, shapeName=dyn_skirt_back_side_r_2_capsule, center=<2 -4.5 0>}]
 dyn_skirt_back_side_r_3 [{shapeType=Capsule, length=32.5053, radius=7, shapeName=dyn_skirt_back_side_r_3_capsule, center=<2 -3.1 0>}]
 dyn_skirt_back_side_r_4 [{shapeType=Capsule, length=32.5053, radius=7, shapeName=dyn_skirt_back_side_r_4_capsule, center=<15 -1.2 0>, rotation=<rot 1.2785 0 0>}]
 dyn_skirt_front_l_1 [{
   shapeType=Capsule
   length=14
   radius=12
   shapeName=dyn_skirt_front_l_1_capsule
   center=<11.2068 -8.952 1.1618>
   rotation=<rot 0.0002 -0.0002 18.4495>
  }]
 dyn_skirt_front_l_2 [{
   shapeType=Capsule
   length=19.4535
   radius=6.5465
   shapeName=dyn_skirt_front_l_2_capsule
   center=<12.4765 0 3.8487>
   rotation=<rot 10.3531 3.1015 28.1221>
  }]
 dyn_skirt_front_l_3 [{
   shapeType=Capsule
   length=19.3344
   radius=6.6656
   shapeName=dyn_skirt_front_l_3_capsule
   center=<16.0238 0 2.0978>
   rotation=<rot 10.7136 3.1742 30.9373>
  }]
 dyn_skirt_front_r_1 [{
   shapeType=Capsule
   length=26.3293
   radius=4.2198
   shapeName=dyn_skirt_front_r_1_capsule
   center=<6.8084 -0.7826 -2.6487>
   rotation=<rot -89.3909 -0.0006 5.1958>
  }]
 dyn_skirt_front_r_2 [{shapeType=Capsule, length=20, radius=6, shapeName=dyn_skirt_front_r_2_capsule, center=<11.1003 0 0>}]
 dyn_skirt_side_l_1 [{shapeType=Capsule, length=20, radius=6, shapeName=dyn_skirt_side_l_1_capsule, center=<2 -4 0>, rotation=<rot 0 0 -7.8265>}]
 dyn_skirt_side_l_2 [{shapeType=Capsule, length=21.4263, radius=4.5737, shapeName=dyn_skirt_side_l_2_capsule, center=<10.6663 0 0>}]
 dyn_skirt_side_l_3 [{shapeType=Capsule, length=19.3335, radius=6.6665, shapeName=dyn_skirt_side_l_3_capsule, center=<15.638 0 0>}]
 dyn_skirt_side_r_1 [{shapeType=Capsule, length=20, radius=6, shapeName=dyn_skirt_side_r_1_capsule, center=<2 -4 0>}]
 dyn_skirt_side_r_2 [{shapeType=Capsule, length=20, radius=6, shapeName=dyn_skirt_side_r_2_capsule, center=<10.6663 -1.5 0>, rotation=<rot 0 0 -25.8162>}]
 dyn_skirt_side_r_3 [{shapeType=Capsule, length=20, radius=6, shapeName=dyn_skirt_side_r_3_capsule, center=<10.4333 0 0>, rotation=<rot 0 0 -18.0954>}]
 dyn_strap_lf_01 [{shapeType=Capsule, length=7.1832, radius=1.3506, shapeName=dyn_strap_lf_01_capsule, center=<5 0 0>, rotation=<rot 90 0 0>}]
 dyn_strap_lf_02 [{shapeType=Capsule, length=6.5328, radius=1.446, shapeName=dyn_strap_lf_02_capsule, center=<5 0 0>, rotation=<rot 90 0 0>}]
 dyn_strap_lf_03 [{shapeType=Capsule, length=6.5296, radius=1.5328, shapeName=dyn_strap_lf_03_capsule, center=<5 0 0>, rotation=<rot 90 0 0>}]
 dyn_strap_lf_04 [{shapeType=Capsule, length=6.9035, radius=1.8276, shapeName=dyn_strap_lf_04_capsule, center=<5 0 0>, rotation=<rot 90 0 0>}]
 dyn_strap_rt_01 [{shapeType=Capsule, length=4.6726, radius=1.1711, shapeName=dyn_strap_rt_01_capsule, center=<5 0 0>, rotation=<rot 90 0 0>}]
 dyn_strap_rt_02 [{shapeType=Capsule, length=7.381, radius=1.3145, shapeName=dyn_strap_rt_02_capsule, center=<5 0 0>, rotation=<rot 90 0 0>}]
 dyn_strap_rt_03 [{shapeType=Capsule, length=6.5175, radius=1.5713, shapeName=dyn_strap_rt_03_capsule, center=<5 0 0>, rotation=<rot 90 0 0>}]
 dyn_strap_rt_04 [{shapeType=Capsule, length=5.6972, radius=1.889, shapeName=dyn_strap_rt_04_capsule, center=<5 0 0>, rotation=<rot 90 0 0>}]
 floor [{shapeType=Capsule, length=5, radius=510, shapeName=floor_capsule, center=<0 0 -505>}]
 thigh_l [{shapeType=Capsule, length=31.8705, radius=9.5295, shapeName=thigh_l_capsule, center=<-15 0.25 -0.73>, rotation=<rot 90 0 0>}]
 thigh_r [{shapeType=Capsule, length=31.8705, radius=9.5295, shapeName=thigh_l_capsule, center=<15 -0.25 0.73>, rotation=<rot 90 0 0>}]
}
constraints [
 {
  bone1Name=dyn_skirt_front_l_2
  bone2Name=dyn_skirt_front_l_1
  swing1LimitDegrees=5
  swing2LimitDegrees=10
  twistLimitDegrees=0
  twistMotion=Locked
 }
 {
  bone1Name=dyn_skirt_front_l_3
  bone2Name=dyn_skirt_front_l_2
  swing1LimitDegrees=5
  swing2LimitDegrees=10
  twistLimitDegrees=0
  twistMotion=Locked
 }
 {
  bone1Name=dyn_skirt_front_r_2
  bone2Name=dyn_skirt_front_r_1
  swing1LimitDegrees=20
  swing2LimitDegrees=10
  twistLimitDegrees=0
  twistMotion=Locked
 }
 {
  bone1Name=dyn_skirt_side_l_2
  bone2Name=dyn_skirt_side_l_1
  swing1LimitDegrees=15
  swing2LimitDegrees=0
  twistLimitDegrees=0
  twistMotion=Locked
 }
 {
  bone1Name=dyn_skirt_side_l_3
  bone2Name=dyn_skirt_side_l_2
  swing1LimitDegrees=15
  swing2LimitDegrees=0
  twistLimitDegrees=0
  twistMotion=Locked
 }
 {
  bone1Name=dyn_skirt_side_r_2
  bone2Name=dyn_skirt_side_r_1
  swing1LimitDegrees=20
  swing2LimitDegrees=10
  twistLimitDegrees=0
  twistMotion=Locked
 }
 {
  bone1Name=dyn_skirt_side_r_3
  bone2Name=dyn_skirt_side_r_2
  swing1LimitDegrees=20
  swing2LimitDegrees=10
  twistLimitDegrees=0
  twistMotion=Locked
 }
 {bone1Name=dyn_loincloth_04, bone2Name=dyn_loincloth_03, swing1LimitDegrees=5, swing2LimitDegrees=30, twistLimitDegrees=1}
 {bone1Name=dyn_loincloth_01, bone2Name=base_loincloth_01, swing1LimitDegrees=5, swing2LimitDegrees=20, twistLimitDegrees=1}
 {bone1Name=dyn_loincloth_02, bone2Name=dyn_loincloth_01, swing1LimitDegrees=5, swing2LimitDegrees=20, twistLimitDegrees=1}
 {bone1Name=dyn_loincloth_03, bone2Name=dyn_loincloth_02, swing1LimitDegrees=5, swing2LimitDegrees=20, twistLimitDegrees=1}
 {bone1Name=dyn_strap_lf_02, bone2Name=dyn_strap_lf_01, swing1LimitDegrees=20, swing2LimitDegrees=20, twistLimitDegrees=45, twistMotion=Locked}
 {bone1Name=dyn_strap_lf_03, bone2Name=dyn_strap_lf_02, swing1LimitDegrees=20, swing2LimitDegrees=20, twistLimitDegrees=45, twistMotion=Locked}
 {bone1Name=dyn_strap_lf_04, bone2Name=dyn_strap_lf_03, swing1LimitDegrees=20, swing2LimitDegrees=20, twistLimitDegrees=45, twistMotion=Locked}
 {bone1Name=dyn_strap_rt_02, bone2Name=dyn_strap_rt_01, swing1LimitDegrees=20, swing2LimitDegrees=20, twistLimitDegrees=45, twistMotion=Locked}
 {bone1Name=dyn_strap_rt_03, bone2Name=dyn_strap_rt_02, swing1LimitDegrees=20, swing2LimitDegrees=20, twistLimitDegrees=45, twistMotion=Locked}
 {bone1Name=dyn_strap_rt_04, bone2Name=dyn_strap_rt_03, swing1LimitDegrees=20, swing2LimitDegrees=20, twistLimitDegrees=45, twistMotion=Locked}
 {
  bone1Name=dyn_skirt_back_0
  bone2Name=base_skirt_back_0
  swing1LimitDegrees=5
  swing2LimitDegrees=5
  swing2Motion=Locked
  twistLimitDegrees=45
  twistMotion=Locked
 }
 {bone1Name=dyn_skirt_back_1, bone2Name=dyn_skirt_back_0, swing1LimitDegrees=10, swing2LimitDegrees=0, twistLimitDegrees=0}
 {bone1Name=dyn_skirt_back_2, bone2Name=dyn_skirt_back_1, swing1LimitDegrees=10, swing2LimitDegrees=0, twistLimitDegrees=0}
 {bone1Name=dyn_skirt_back_3, bone2Name=dyn_skirt_back_2, swing1LimitDegrees=10, swing2LimitDegrees=0, twistLimitDegrees=0}
 {bone1Name=dyn_skirt_back_4, bone2Name=dyn_skirt_back_3, swing1LimitDegrees=10, swing2LimitDegrees=0, twistLimitDegrees=0}
 {
  bone1Name=dyn_skirt_back_side_l_0
  bone2Name=base_skirt_back_side_l_0
  swing1LimitDegrees=5
  swing2LimitDegrees=5
  swing2Motion=Locked
  twistLimitDegrees=45
  twistMotion=Locked
 }
 {bone1Name=dyn_skirt_back_side_l_1, bone2Name=dyn_skirt_back_side_l_0, swing1LimitDegrees=10, swing2LimitDegrees=0, twistLimitDegrees=0}
 {bone1Name=dyn_skirt_back_side_l_2, bone2Name=dyn_skirt_back_side_l_1, swing1LimitDegrees=10, swing2LimitDegrees=0, twistLimitDegrees=0}
 {bone1Name=dyn_skirt_back_side_l_3, bone2Name=dyn_skirt_back_side_l_2, swing1LimitDegrees=10, swing2LimitDegrees=0, twistLimitDegrees=0}
 {bone1Name=dyn_skirt_back_side_l_4, bone2Name=dyn_skirt_back_side_l_3, swing1LimitDegrees=10, swing2LimitDegrees=0, twistLimitDegrees=0}
 {
  bone1Name=dyn_skirt_back_side_r_0
  bone2Name=base_skirt_back_side_r_0
  swing1LimitDegrees=5
  swing2LimitDegrees=5
  swing2Motion=Locked
  twistLimitDegrees=45
  twistMotion=Locked
 }
 {bone1Name=dyn_skirt_back_side_r_1, bone2Name=dyn_skirt_back_side_r_0, swing1LimitDegrees=10, swing2LimitDegrees=0, twistLimitDegrees=0}
 {bone1Name=dyn_skirt_back_side_r_2, bone2Name=dyn_skirt_back_side_r_1, swing1LimitDegrees=10, swing2LimitDegrees=0, twistLimitDegrees=0}
 {bone1Name=dyn_skirt_back_side_r_3, bone2Name=dyn_skirt_back_side_r_2, swing1LimitDegrees=10, swing2LimitDegrees=0, twistLimitDegrees=0}
 {bone1Name=dyn_skirt_back_side_r_4, bone2Name=dyn_skirt_back_side_r_3, swing1LimitDegrees=10, swing2LimitDegrees=0, twistLimitDegrees=0}
 {
  bone1Name=dyn_skirt_back_4
  bone2Name=dyn_skirt_back_side_l_4
  swing1LimitDegrees=45
  swing1Motion=Locked
  swing2LimitDegrees=45
  swing2Motion=Locked
  twistLimitDegrees=45
  twistMotion=Locked
 }
 {
  bone1Name=dyn_skirt_back_4
  bone2Name=dyn_skirt_back_side_r_4
  swing1LimitDegrees=45
  swing1Motion=Locked
  swing2LimitDegrees=45
  swing2Motion=Locked
  twistLimitDegrees=45
  twistMotion=Locked
 }
 {
  bone1Name=dyn_skirt_front_l_3
  bone2Name=dyn_skirt_front_l_2
  swing1LimitDegrees=45
  swing1Motion=Free
  swing2LimitDegrees=45
  swing2Motion=Free
  twistLimitDegrees=45
  twistMotion=Free
 }
]
