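# Post-process style AnimBP for the M_MED_SubtleSilly player body cosmetic, adding corrective and secondary motion to the pose copied from its attached parent.
# Cached poses chain a movement-state-driven Control Rig, loincloth and backpack-attach offsets, then dynamics.
# Dynamics mixes AnimDynamics, two RigidBody sims, constraints and additive bone fixes across sleeves, skirt, cape, chains, strap, ring and sword.
# EventGraph derives gravity override, curve-based values, movement flags and matching-backpack detection.
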AnimBlueprint /BRCosmetics/Characters/Player/Male/Medium/Bodies/M_MED_SubtleSilly/Meshes/M_MED_SubtleSilly_AnimBP
target_skeleton=&/M_MED_SubtleSilly_Skeleton
graphs [
 AnimGraph {nodes [
   Output Pose(Result=Use cached pose 'DynamicsCache'()) {layerGroup=None}
   n1 = Control Rig!(
    Source=Copy Pose From Mesh() {bUseAttachedParent=true}
    Alpha
    isCrouchingCustom:bool=Get crouching wt sliding() -> (crouching wt sliding:bool)
    isSprinting:bool=Get sprinting modified() -> (sprinting modified:bool)
    isSlidingKnee:bool=Get bIsKneeSliding()
    isSlidingBaseball:bool=Get bIsBaseballSliding()
    isTargeting:bool=Get bIsTargeting()
   ) -> (Pose) {controlRigAssetReference {controlRigAsset=&/M_MED_SubtleSilly_CtrlRig}}
   Save cached pose 'controlRigCache'(Pose=@n1) {cacheName=controlRigCache}
   n2 = Anim Dynamics - Bone: dyn_sleeve_rt!(
    ComponentPose=Local To Component(LocalPose=Use cached pose 'modifyTransformsCache'())
    Alpha=1
    SimSpaceSettings:Anim Phys Sim Space Settings Structure=(SimSpaceAngularAlpha=0.5,MaxAngularVelocity=7,MaxAngularAcceleration=45,ExternalAngularVelocity=(X=0,Y=0,Z=0))
   ) -> (Pose) {
    angularBiasOverride=1
    angularDampingOverride=1
    angularSpringConstant=20000
    bAngularSpring=true
    bOverrideAngularBias=true
    bOverrideAngularDamping=true
    bUseGravityOverride=true
    gravityScale=0.35
    simulationSpace=RootRelative
    boundBone {boneName=dyn_sleeve_rt}
    componentAppliedLinearAccClamp=<3000 3000 3000>
    gravityOverride {z=-980}
    physicsBodyDefinitions [{
      boundBone {boneName=dyn_sleeve_rt}
      boxExtents=<30 30 30>
      constraintSetup {angularLimitsMax=<0 15 15>, angularLimitsMin=<0 -15 -15>, angularTarget=<1 0 0>}
      localJointOffset=<8 0 0>
     }]
    simSpaceSettings {maxAngularAcceleration=45, maxAngularVelocity=7, simSpaceAngularAlpha=0.5}
   }
   n3 = Anim Dynamics - Bone: dyn_sleeve_lf!(
    ComponentPose=@n2
    Alpha=1
    SimSpaceSettings:Anim Phys Sim Space Settings Structure=(SimSpaceAngularAlpha=0.5,MaxAngularVelocity=7,MaxAngularAcceleration=45,ExternalAngularVelocity=(X=0,Y=0,Z=0))
   ) -> (Pose) {
    angularBiasOverride=1
    angularDampingOverride=1
    angularSpringConstant=20000
    bAngularSpring=true
    bOverrideAngularBias=true
    bOverrideAngularDamping=true
    bUseGravityOverride=true
    gravityScale=0.35
    simulationSpace=RootRelative
    boundBone {boneName=dyn_sleeve_lf}
    componentAppliedLinearAccClamp=<3000 3000 3000>
    gravityOverride {z=-980}
    physicsBodyDefinitions [{
      boundBone {boneName=dyn_sleeve_lf}
      boxExtents=<30 30 30>
      constraintSetup {angularLimitsMax=<0 15 15>, angularLimitsMin=<0 -15 -15>, angularTarget=<1 0 0>}
      localJointOffset=<8 0 0>
     }]
    simSpaceSettings {maxAngularAcceleration=45, maxAngularVelocity=7, simSpaceAngularAlpha=0.5}
   }
   n4 = RigidBody!(ComponentPose=@n3, Alpha, OverrideWorldGravity:Vector="0,0,0") -> (Pose) {
    bClampLinearTranslationLimitToRefPose=true
    bOverrideWorldGravity=true
    overridePhysicsAsset=&/M_MED_SubtleSilly_Physics
    simulationSpace=BaseBoneSpace
    baseBoneRef {boneName=pelvis}
    simSpaceSettings {
     dampingAlpha=0
     maxAngularAcceleration=9
     maxAngularVelocity=4
     maxLinearAcceleration=1000
     maxLinearVelocity=400
     velocityScaleZ=0.75
     worldAlpha=1
     externalLinearDragV=<1.1 1.1 1.1>
    }
   }
   n5 = "Constraint - base_chain_0 (RR)"!(ComponentPose=@n4, Alpha=1, ConstraintWeights_0=0.6, ConstraintWeights_1:float=0.4) -> (Pose) {
    boneToModify {boneName=base_chain_0}
    constraintSetup [{transformType=Rotation, targetBone {boneName=dyn_skirt_back_side_r_0}}, {transformType=Rotation, targetBone {boneName=dyn_skirt_side_r_0}}]
    constraintWeights [1,1]
   }
   n6 = "Constraint - base_chain_3 (RR)"!(ComponentPose=@n5, Alpha=1, ConstraintWeights_0=0.5, ConstraintWeights_1:float=0.5) -> (Pose) {
    boneToModify {boneName=base_chain_3}
    constraintSetup [{transformType=Rotation, targetBone {boneName=dyn_skirt_front_r_0}}, {transformType=Rotation, targetBone {boneName=dyn_skirt_front_0}}]
    constraintWeights [1,1]
   }
   n7 = "Transform (Modify) Bone - Bone: base_chain_3"!(
    ComponentPose="Transform (Modify) Bone - Bone: base_chain_0"!(ComponentPose=@n6, Alpha=1, Rotation="0,-4,0") -> (Pose) {rotationMode=BMM_Additive, rotationSpace=BCS_BoneSpace, boneToModify {boneName=base_chain_0}, rotation {yaw=-4}}
    Alpha=1
    Translation="-0.306,-1.149,-0.167"
    Rotation="26.662,-9.715,8.515"
   ) -> (Pose) {
    rotationMode=BMM_Additive
    rotationSpace=BCS_BoneSpace
    translationMode=BMM_Additive
    translationSpace=BCS_BoneSpace
    boneToModify {boneName=base_chain_3}
    rotation=<rot 26.662 -9.715 8.515>
    translation=<-0.306 -1.149 -0.167>
   }
   n8 = "Transform (Modify) Bone - Bone: base_chain_3"!(ComponentPose=@n7, Alpha=1, Rotation="-3.686,-19.717,-13.965") -> (Pose) {
    rotationMode=BMM_Additive
    rotationSpace=BCS_BoneSpace
    translationMode=BMM_Additive
    translationSpace=BCS_BoneSpace
    boneToModify {boneName=base_chain_3}
    rotation=<rot -3.686 -19.717 -13.965>
   }
   n9 = "Transform (Modify) Bone - Bone: base_chain_0"(ComponentPose=@n8, Alpha=1, Rotation="-8.867,-31.126,13.514") {
    rotationMode=BMM_Additive
    rotationSpace=BCS_BoneSpace
    translationMode=BMM_Additive
    translationSpace=BCS_BoneSpace
    boneToModify {boneName=base_chain_0}
    rotation=<rot -8.867 -31.126 13.514>
   }
   n10 = "Transform (Modify) Bone - Bone: base_chain_3"!(ComponentPose=@n9, bAlphaBoolEnabled:bool=True, Rotation="5.605,-9.089,22.123") -> (Pose) {
    alphaInputType=Bool
    rotationMode=BMM_Additive
    rotationSpace=BCS_BoneSpace
    translationSpace=BCS_BoneSpace
    boneToModify {boneName=base_chain_3}
    rotation=<rot 5.605 -9.089 22.123>
   }
   n11 = "Transform (Modify) Bone - Bone: base_chain_0"!(ComponentPose=@n10, bAlphaBoolEnabled:bool=True, Rotation="1.731,7.67,13.867") -> (Pose) {
    alphaInputType=Bool
    rotationMode=BMM_Additive
    rotationSpace=BCS_BoneSpace
    translationSpace=BCS_BoneSpace
    boneToModify {boneName=base_chain_0}
    rotation=<rot 1.731 7.67 13.867>
   }
   n12 = "Transform (Modify) Bone - Bone: base_chain_0"!(ComponentPose=@n11, Alpha:float=1, Translation:Vector="0,0,0", Rotation="0,-5.5,0") -> (Pose) {
    rotationMode=BMM_Additive
    rotationSpace=BCS_BoneSpace
    translationMode=BMM_Additive
    translationSpace=BCS_BoneSpace
    boneToModify {boneName=base_chain_0}
    rotation {yaw=-5.5}
   }
   n13 = "Transform (Modify) Bone - Bone: base_chain_3"!(ComponentPose=@n12, bAlphaBoolEnabled=True, Translation:Vector="1.124,0.044,2.494", Rotation="10.965,-11.029,-29.076") -> (Pose) {
    alphaInputType=Bool
    rotationMode=BMM_Additive
    rotationSpace=BCS_BoneSpace
    translationMode=BMM_Additive
    translationSpace=BCS_BoneSpace
    boneToModify {boneName=base_chain_3}
    rotation=<rot 10.965 -11.029 -29.076>
    translation=<1.124 0.044 2.494>
   }
   n14 = "Transform (Modify) Bone - Bone: base_chain_0"!(ComponentPose=@n13, bAlphaBoolEnabled:bool=True, Rotation="-12.966,0.427,33.063") -> (Pose) {
    alphaInputType=Bool
    rotationMode=BMM_Additive
    rotationSpace=BCS_BoneSpace
    translationMode=BMM_Additive
    translationSpace=BCS_BoneSpace
    boneToModify {boneName=base_chain_0}
    rotation=<rot -12.966 0.427 33.063>
   }
   n15 = "Transform (Modify) Bone - Bone: base_chain_3"(ComponentPose=@n14, bAlphaBoolEnabled=True, Translation="0.257,1.673,0.845", Rotation="-28.855,-20.572,-61.672") {
    alphaInputType=Bool
    rotationMode=BMM_Additive
    rotationSpace=BCS_BoneSpace
    translationMode=BMM_Additive
    translationSpace=BCS_BoneSpace
    boneToModify {boneName=base_chain_3}
    rotation=<rot -28.855 -20.572 -61.672>
    translation=<0.257 1.673 0.845>
   }
   n16 = "Transform (Modify) Bone - Bone: base_chain_0"(ComponentPose=@n15, bAlphaBoolEnabled=True, Rotation="-40.554,-9.906,35.414") {
    alphaInputType=Bool
    rotationMode=BMM_Additive
    rotationSpace=BCS_BoneSpace
    translationMode=BMM_Additive
    translationSpace=BCS_BoneSpace
    boneToModify {boneName=base_chain_0}
    rotation=<rot -40.554 -9.906 35.414>
   }
   n17 = "Transform (Modify) Bone - Bone: dyn_sword"!(ComponentPose=@n16, bAlphaBoolEnabled:bool=True, Rotation="-8.670,-10.969,-32.997") -> (Pose) {
    alphaInputType=Bool
    rotationMode=BMM_Additive
    rotationSpace=BCS_BoneSpace
    translationMode=BMM_Additive
    translationSpace=BCS_BoneSpace
    boneToModify {boneName=dyn_sword}
    rotation=<rot -8.67 -10.969 -32.997>
   }
   n18 = Anim Dynamics - Bone: dyn_chain_1!(
    ComponentPose=@n17
    Alpha=1
    GravityOverride:Vector="0,0,-980"
    SimSpaceSettings:Anim Phys Sim Space Settings Structure=(SimSpaceAngularAlpha=0.3,MaxAngularVelocity=7,MaxAngularAcceleration=35,ExternalAngularVelocity=(X=0,Y=0,Z=0))
    ExternalForce="0,0,0"
   ) -> (Pose) {
    angularDampingOverride=1
    angularSpringConstant=200
    bOverrideAngularDamping=true
    bUseGravityOverride=true
    bUsePlanarLimit=false
    gravityScale=0.35
    simulationSpace=RootRelative
    boundBone {boneName=dyn_chain_1}
    componentAppliedLinearAccClamp=<3000 3000 3000>
    componentLinearAccScale=<0.5 0.5 0.5>
    componentLinearVelScale=<0.5 0.5 0.5>
    gravityOverride {z=-980}
    physicsBodyDefinitions [{
      boundBone {boneName=dyn_chain_1}
      boxExtents=<20 20 20>
      constraintSetup {angularLimitsMax=<0 15 0>, angularLimitsMin=<0 -5 -15>, angularTarget=<1 0 0>}
      localJointOffset=<8 0 0>
     }]
    relativeSpaceBone {boneName=pelvis}
    simSpaceSettings {maxAngularAcceleration=35, maxAngularVelocity=7, simSpaceAngularAlpha=0.3}
   }
   n19 = Anim Dynamics - Bone: dyn_chain_2!(
    ComponentPose=@n18
    Alpha=1
    GravityOverride:Vector="0,0,-980"
    SimSpaceSettings:Anim Phys Sim Space Settings Structure=(SimSpaceAngularAlpha=0.3,MaxAngularVelocity=7,MaxAngularAcceleration=35,ExternalAngularVelocity=(X=0,Y=0,Z=0))
    ExternalForce="0,0,0"
   ) -> (Pose) {
    angularDampingOverride=1
    angularSpringConstant=200
    bOverrideAngularDamping=true
    bUseGravityOverride=true
    bUsePlanarLimit=false
    gravityScale=0.35
    simulationSpace=RootRelative
    boundBone {boneName=dyn_chain_2}
    componentAppliedLinearAccClamp=<3000 3000 3000>
    componentLinearAccScale=<0.5 0.5 0.5>
    componentLinearVelScale=<0.5 0.5 0.5>
    gravityOverride {z=-980}
    physicsBodyDefinitions [{
      boundBone {boneName=dyn_chain_2}
      boxExtents=<20 20 20>
      constraintSetup {angularLimitsMax=<0 15 0>, angularLimitsMin=<0 -5 -15>, angularTarget=<1 0 0>}
      localJointOffset=<8 0 0>
     }]
    relativeSpaceBone {boneName=pelvis}
    simSpaceSettings {maxAngularAcceleration=35, maxAngularVelocity=7, simSpaceAngularAlpha=0.3}
   }
   n20 = Anim Dynamics - Bone: dyn_chain_4!(
    ComponentPose=@n19
    Alpha=1
    GravityOverride:Vector="0,0,-980"
    SimSpaceSettings:Anim Phys Sim Space Settings Structure=(SimSpaceAngularAlpha=0.3,MaxAngularVelocity=7,MaxAngularAcceleration=35,ExternalAngularVelocity=(X=0,Y=0,Z=0))
    ExternalForce="0,0,0"
   ) -> (Pose) {
    angularDampingOverride=1
    angularSpringConstant=200
    bOverrideAngularDamping=true
    bUseGravityOverride=true
    gravityScale=0.35
    simulationSpace=RootRelative
    boundBone {boneName=dyn_chain_4}
    componentAppliedLinearAccClamp=<3000 3000 3000>
    componentLinearAccScale=<0.5 0.5 0.5>
    componentLinearVelScale=<0.5 0.5 0.5>
    gravityOverride {z=-980}
    physicsBodyDefinitions [{
      boundBone {boneName=dyn_chain_4}
      boxExtents=<20 20 20>
      constraintSetup {angularLimitsMax=<0 15 0>, angularLimitsMin=<0 -5 -15>, angularTarget=<1 0 0>}
      localJointOffset=<8 0 0>
     }]
    relativeSpaceBone {boneName=pelvis}
    simSpaceSettings {maxAngularAcceleration=35, maxAngularVelocity=7, simSpaceAngularAlpha=0.3}
   }
   n21 = Anim Dynamics - Bone: dyn_chain_5!(
    ComponentPose=@n20
    Alpha=1
    GravityOverride:Vector="0,0,-980"
    SimSpaceSettings:Anim Phys Sim Space Settings Structure=(SimSpaceAngularAlpha=0.3,MaxAngularVelocity=7,MaxAngularAcceleration=35,ExternalAngularVelocity=(X=0,Y=0,Z=0))
    ExternalForce="0,0,0"
   ) -> (Pose) {
    angularDampingOverride=1
    angularSpringConstant=200
    bOverrideAngularDamping=true
    bUseGravityOverride=true
    gravityScale=0.35
    simulationSpace=RootRelative
    boundBone {boneName=dyn_chain_5}
    componentAppliedLinearAccClamp=<3000 3000 3000>
    componentLinearAccScale=<0.5 0.5 0.5>
    componentLinearVelScale=<0.5 0.5 0.5>
    gravityOverride {z=-980}
    physicsBodyDefinitions [{
      boundBone {boneName=dyn_chain_5}
      boxExtents=<20 20 20>
      constraintSetup {angularLimitsMax=<0 15 0>, angularLimitsMin=<0 -5 -15>, angularTarget=<1 0 0>}
      localJointOffset=<8 0 0>
     }]
    relativeSpaceBone {boneName=pelvis}
    simSpaceSettings {maxAngularAcceleration=35, maxAngularVelocity=7, simSpaceAngularAlpha=0.3}
   }
   n22 = Anim Dynamics - Bone: dyn_strap_rt_02!(
    ComponentPose=@n21
    Alpha=0.5
    GravityOverride:Vector="0,0,-980"
    SimSpaceSettings:Anim Phys Sim Space Settings Structure=(SimSpaceAngularAlpha=1,MaxAngularVelocity=7,MaxAngularAcceleration=35,ExternalAngularVelocity=(X=0,Y=0,Z=0))
    ExternalForce="0,0,0"
   ) -> (Pose) {
    angularDampingOverride=1
    angularSpringConstant=20000
    bOverrideAngularDamping=true
    bUseGravityOverride=true
    gravityScale=0.35
    simulationSpace=RootRelative
    alphaScaleBiasClamp {bInterpResult=true}
    boundBone {boneName=dyn_strap_rt_02}
    componentAppliedLinearAccClamp=<3000 3000 3000>
    componentLinearAccScale=<0.3 0.3 0.3>
    componentLinearVelScale=<0.3 0.3 0.3>
    gravityOverride {z=-980}
    physicsBodyDefinitions [{
      boundBone {boneName=dyn_strap_rt_02}
      boxExtents=<50 50 50>
      constraintSetup {angularLimitsMax=<0 5 30>, angularLimitsMin=<0 -20 -30>, angularTarget=<1 0 0>}
      localJointOffset=<8 0 0>
     }]
    simSpaceSettings {maxAngularAcceleration=35, maxAngularVelocity=7, simSpaceAngularAlpha=1}
   }
   n23 = "Transform (Modify) Bone - Bone: dyn_skirt_back_side_l_2"!(ComponentPose=@n22, bAlphaBoolEnabled:bool=True, Rotation="-27.807,-14.415,-7.541") -> (Pose) {
    alphaInputType=Bool
    rotationMode=BMM_Additive
    rotationSpace=BCS_BoneSpace
    translationMode=BMM_Additive
    translationSpace=BCS_BoneSpace
    boneToModify {boneName=dyn_skirt_back_side_l_2}
    rotation=<rot -27.807 -14.415 -7.541>
    translation=<-0.2455 -0.7528 -0.0931>
   }
   n24 = "Transform (Modify) Bone - Bone: dyn_skirt_back_1"!(ComponentPose=@n23, bAlphaBoolEnabled:bool=True, Translation="0.011,-3.493,0.002", Rotation="-5.784,-8.170,-9.185") -> (Pose) {
    alphaInputType=Bool
    rotationMode=BMM_Additive
    rotationSpace=BCS_BoneSpace
    translationMode=BMM_Additive
    translationSpace=BCS_BoneSpace
    boneToModify {boneName=dyn_skirt_back_1}
    rotation=<rot -5.784 -8.17 -9.185>
    translation=<0.011 -3.493 0.002>
   }
   n25 = "Transform (Modify) Bone - Bone: dyn_skirt_back_side_r_2"!(ComponentPose=@n24, bAlphaBoolEnabled:bool=True, Rotation="0.000,-7.538,0.000") -> (Pose) {
    alphaInputType=Bool
    rotationMode=BMM_Additive
    rotationSpace=BCS_BoneSpace
    translationSpace=BCS_BoneSpace
    boneToModify {boneName=dyn_skirt_back_side_r_2}
    rotation {yaw=-7.538}
   }
   n26 = "Transform (Modify) Bone - Bone: dyn_skirt_front_l_3"!(ComponentPose=@n25, bAlphaBoolEnabled:bool=True, Rotation="0.218,-12.165,19.614") -> (Pose) {
    alphaInputType=Bool
    rotationMode=BMM_Additive
    rotationSpace=BCS_BoneSpace
    boneToModify {boneName=dyn_skirt_front_l_3}
    rotation=<rot 0.218 -12.165 19.614>
   }
   n27 = "Transform (Modify) Bone - Bone: dyn_skirt_side_l_3"!(ComponentPose=@n26, bAlphaBoolEnabled:bool=True, Rotation="1.959,-10.056,16.353") -> (Pose) {
    alphaInputType=Bool
    rotationMode=BMM_Additive
    rotationSpace=BCS_BoneSpace
    boneToModify {boneName=dyn_skirt_side_l_3}
    rotation=<rot 1.959 -10.056 16.353>
   }
   n28 = RigidBody(ComponentPose=@n27, OverrideWorldGravity="0,0,0") {
    bClampLinearTranslationLimitToRefPose=true
    bOverrideWorldGravity=true
    overridePhysicsAsset=&/M_MED_SubtleSilly_Cape_Physics
    simulationSpace=BaseBoneSpace
    baseBoneRef {boneName=spine_03}
    simSpaceSettings {dampingAlpha=0, maxAngularAcceleration=10, maxAngularVelocity=5, maxLinearAcceleration=1200, maxLinearVelocity=500, worldAlpha=1}
   }
   n29 = Anim Dynamics - Bone: dyn_cape_03!(
    ComponentPose=@n28
    Alpha=1
    GravityOverride:Vector="0,0,-980"
    SimSpaceSettings:Anim Phys Sim Space Settings Structure=(SimSpaceAngularAlpha=0.55,MaxAngularVelocity=7,MaxAngularAcceleration=35,ExternalAngularVelocity=(X=0,Y=0,Z=0))
   ) -> (Pose) {
    angularBiasOverride=1
    angularDampingOverride=1
    angularSpringConstant=10000
    bAngularSpring=true
    bGravityOverrideInSimSpace=true
    bOverrideAngularDamping=true
    bUseGravityOverride=true
    gravityScale=0.35
    linearSpringConstant=10000
    simulationSpace=RootRelative
    boundBone {boneName=dyn_cape_03}
    componentAppliedLinearAccClamp=<3000 3000 3000>
    gravityOverride {z=-980}
    physicsBodyDefinitions [{
      boundBone {boneName=dyn_cape_03}
      boxExtents=<40 40 40>
      constraintSetup {angularLimitsMax=<0 1 10>, angularLimitsMin=<0 -10 -10>, angularTarget=<1 0 0>}
      localJointOffset=<8 0 0>
     }]
    relativeSpaceBone {boneName=clavicle_l}
    simSpaceSettings {maxAngularAcceleration=35, maxAngularVelocity=7, simSpaceAngularAlpha=0.55}
   }
   n30 = Anim Dynamics - Bone: dyn_cape_04!(
    ComponentPose=@n29
    Alpha=1
    GravityOverride:Vector="0,0,-980"
    SimSpaceSettings:Anim Phys Sim Space Settings Structure=(SimSpaceAngularAlpha=0.55,MaxAngularVelocity=7,MaxAngularAcceleration=35,ExternalAngularVelocity=(X=0,Y=0,Z=0))
   ) -> (Pose) {
    angularBiasOverride=1
    angularDampingOverride=1
    angularSpringConstant=10000
    bAngularSpring=true
    bGravityOverrideInSimSpace=true
    bOverrideAngularDamping=true
    bUseGravityOverride=true
    gravityScale=0.35
    linearSpringConstant=10000
    simulationSpace=RootRelative
    boundBone {boneName=dyn_cape_04}
    componentAppliedLinearAccClamp=<3000 3000 3000>
    gravityOverride {z=-980}
    physicsBodyDefinitions [{
      boundBone {boneName=dyn_cape_04}
      boxExtents=<40 40 40>
      constraintSetup {angularLimitsMax=<0 1 10>, angularLimitsMin=<0 -10 -10>, angularTarget=<1 0 0>}
      localJointOffset=<8 0 0>
     }]
    relativeSpaceBone {boneName=clavicle_l}
    simSpaceSettings {maxAngularAcceleration=35, maxAngularVelocity=7, simSpaceAngularAlpha=0.55}
   }
   n31 = Anim Dynamics - Bone: dyn_cape_05!(
    ComponentPose=@n30
    Alpha=1
    GravityOverride:Vector="0,0,-980"
    SimSpaceSettings:Anim Phys Sim Space Settings Structure=(SimSpaceAngularAlpha=0.55,MaxAngularVelocity=7,MaxAngularAcceleration=35,ExternalAngularVelocity=(X=0,Y=0,Z=0))
   ) -> (Pose) {
    angularBiasOverride=1
    angularDampingOverride=1
    angularSpringConstant=10000
    bAngularSpring=true
    bGravityOverrideInSimSpace=true
    bOverrideAngularDamping=true
    bUseGravityOverride=true
    gravityScale=0.35
    linearSpringConstant=10000
    simulationSpace=RootRelative
    boundBone {boneName=dyn_cape_05}
    componentAppliedLinearAccClamp=<3000 3000 3000>
    gravityOverride {z=-980}
    physicsBodyDefinitions [{
      boundBone {boneName=dyn_cape_05}
      boxExtents=<40 40 40>
      constraintSetup {angularLimitsMax=<0 1 10>, angularLimitsMin=<0 -10 -10>, angularTarget=<1 0 0>}
      localJointOffset=<8 0 0>
     }]
    relativeSpaceBone {boneName=clavicle_l}
    simSpaceSettings {maxAngularAcceleration=35, maxAngularVelocity=7, simSpaceAngularAlpha=0.55}
   }
   n32 = Anim Dynamics - Bone: dyn_cape_06!(
    ComponentPose=@n31
    Alpha=1
    GravityOverride:Vector="0,0,-980"
    SimSpaceSettings:Anim Phys Sim Space Settings Structure=(SimSpaceAngularAlpha=0.55,MaxAngularVelocity=7,MaxAngularAcceleration=35,ExternalAngularVelocity=(X=0,Y=0,Z=0))
   ) -> (Pose) {
    angularBiasOverride=1
    angularDampingOverride=1
    angularSpringConstant=10000
    bAngularSpring=true
    bGravityOverrideInSimSpace=true
    bOverrideAngularDamping=true
    bUseGravityOverride=true
    gravityScale=0.35
    linearSpringConstant=10000
    simulationSpace=RootRelative
    boundBone {boneName=dyn_cape_06}
    componentAppliedLinearAccClamp=<3000 3000 3000>
    gravityOverride {z=-980}
    physicsBodyDefinitions [{
      boundBone {boneName=dyn_cape_06}
      boxExtents=<40 40 40>
      constraintSetup {angularLimitsMax=<0 1 10>, angularLimitsMin=<0 -10 -10>, angularTarget=<1 0 0>}
      localJointOffset=<8 0 0>
     }]
    relativeSpaceBone {boneName=clavicle_l}
    simSpaceSettings {maxAngularAcceleration=35, maxAngularVelocity=7, simSpaceAngularAlpha=0.55}
   }
   n33 = "Constraint - base_ring (PPPPPPP)"!(
    ComponentPose=@n32
    Alpha=1
    ConstraintWeights_0=0.677
    ConstraintWeights_1:float=0.1129
    ConstraintWeights_2:float=0.1113
    ConstraintWeights_3:float=0.0983
    ConstraintWeights_4:float=0.0002
    ConstraintWeights_5:float=0.0002
    ConstraintWeights_6:float=0.0001
   ) -> (Pose) {
    boneToModify {boneName=base_ring}
    constraintSetup [
     {targetBone {boneName=dyn_skirt_back_0}}
     {targetBone {boneName=dyn_skirt_back_side_l_0}}
     {targetBone {boneName=dyn_skirt_back_side_r_0}}
     {targetBone {boneName=pelvis}}
     {targetBone {boneName=spine_01}}
     {targetBone {boneName=spine_02}}
     {targetBone {boneName=spine_03}}
    ]
    constraintWeights [1,1,1,1,1,1,1]
   }
   n34 = Anim Dynamics - Bone: dyn_ring!(ComponentPose=@n33, bAlphaBoolEnabled:bool=True) -> (Pose) {
    alphaInputType=Bool
    angularDampingOverride=1
    angularSpringConstant=20000
    bOverrideAngularDamping=true
    bUseGravityOverride=true
    gravityScale=0.35
    simulationSpace=RootRelative
    boundBone {boneName=dyn_ring}
    componentAppliedLinearAccClamp=<3000 3000 3000>
    componentLinearAccScale=<0.3 0.3 0.3>
    componentLinearVelScale=<0.3 0.3 0.3>
    gravityOverride {z=-980}
    physicsBodyDefinitions [{
      boundBone {boneName=dyn_ring}
      boxExtents=<20 10 10>
      constraintSetup {angularLimitsMax=<30 30 30>, angularLimitsMin=<-30 -30 -30>, angularTarget=<1 0 0>}
      localJointOffset=<8 0 0>
     }]
    relativeSpaceBone {boneName=pelvis}
    simSpaceSettings {maxAngularAcceleration=35, maxAngularVelocity=7, simSpaceAngularAlpha=0.5}
   }
   n35 = Anim Dynamics - Bone: dyn_ring(ComponentPose=@n34, bAlphaBoolEnabled=True) {
    alphaInputType=Bool
    angularDampingOverride=1
    angularSpringConstant=20000
    bOverrideAngularDamping=true
    bUseGravityOverride=true
    gravityScale=0.35
    simulationSpace=RootRelative
    boundBone {boneName=dyn_ring}
    componentAppliedLinearAccClamp=<100 100 100>
    componentLinearAccScale=<0.1 0.1 0.1>
    componentLinearVelScale=<0.1 0.1 0.1>
    gravityOverride {z=-980}
    physicsBodyDefinitions [{
      boundBone {boneName=dyn_ring}
      boxExtents=<40 20 20>
      constraintSetup {angularLimitsMax=<20 20 20>, angularLimitsMin=<-20 -20 -20>, angularTarget=<1 0 0>}
      localJointOffset=<8 0 0>
     }]
    relativeSpaceBone {boneName=pelvis}
    simSpaceSettings {maxAngularAcceleration=35, maxAngularVelocity=7, simSpaceAngularAlpha=0.5}
   }
   n36 = "Transform (Modify) Bone - Bone: base_sword"!(ComponentPose=@n35, bAlphaBoolEnabled:bool=True, Rotation="10,10.5,12") -> (Pose) {
    alphaInputType=Bool
    rotationMode=BMM_Additive
    rotationSpace=BCS_BoneSpace
    translationMode=BMM_Additive
    translationSpace=BCS_BoneSpace
    boneToModify {boneName=base_sword}
    rotation=<rot 10 10.5 12>
   }
   n37 = "Transform (Modify) Bone - Bone: base_sleeve_rt"!(ComponentPose=@n36, bAlphaBoolEnabled:bool=True, Rotation="-30,0,0") -> (Pose) {alphaInputType=Bool, rotationMode=BMM_Additive, rotationSpace=BCS_BoneSpace, boneToModify {boneName=base_sleeve_rt}, rotation {pitch=-30}}
   Save cached pose 'DynamicsCache'(Pose=Component To Local(ComponentPose=@n37)) {cacheName=DynamicsCache}
   n38 = "Transform (Modify) Bone - Bone: base_loincloth_01"!(
    ComponentPose=Local To Component(LocalPose=Use cached pose 'controlRigCache'())
    bAlphaBoolEnabled:bool=True
    Translation="-2.160,-3.352,0.401"
   ) -> (Pose) {
    alphaInputType=Bool
    rotationMode=BMM_Additive
    rotationSpace=BCS_BoneSpace
    translationMode=BMM_Additive
    translationSpace=BCS_BoneSpace
    alphaBoolBlend {blendInTime=0.5, blendOutTime=0.5}
    boneToModify {boneName=base_loincloth_01}
    translation=<-2.16 -3.352 0.401>
   }
   n39 = "Transform (Modify) Bone - Bone: attach_backpack"!(ComponentPose=@n38, bAlphaBoolEnabled:bool=True, Translation="-2,3,0", Rotation="0,2,-2") -> (Pose) {
    alphaInputType=Bool
    rotationMode=BMM_Additive
    rotationSpace=BCS_BoneSpace
    translationMode=BMM_Additive
    translationSpace=BCS_BoneSpace
    boneToModify {boneName=attach_backpack}
    rotation {roll=-2, yaw=2}
    translation=<-2 3>
   }
   Save cached pose 'modifyTransformsCache'(Pose=Component To Local(ComponentPose=@n39)) {cacheName=modifyTransformsCache}
  ]}
 EventGraph {nodes [
   Event BlueprintUpdateAnimation()
   > n1 = Sequence()
   > n2 = "Gravity Override State Function (Parameter Asset) (Nativized)"(
    CharacterPartAnimInstance=Self-Reference()
    GameTimeInSeconds=GetGameTimeInSeconds()
    Parameters=&/M_MED_SubtleSilly_Parameters
   )
   > Set GravityOverride(GravityOverride=@n2)
   n3 = GetCurveValue(CurveName=thigh_up)
   > Set thigh up(thigh up:double=MapRangeClamped(Value=float - float(A=1, B=@n3), InRangeA=0, InRangeB=0.7)) -> (Output_Get:double)
   n4 = GetCurveValue(CurveName=thigh_side)
   > Set thigh side(thigh side:double=MapRangeClamped(Value=float - float(A=1, B=@n4), InRangeA=0, InRangeB=1)) -> (Output_Get:double)
   > n5 = PlayerMovement(AnimInstance=Self-Reference(), XDotProductCoefficient=100, YDotProductCoefficient=100, ZDotProductCoefficient=100)
   n6 = BreakVector(InVec=MakeVector(X=@n5, Y=0))
   > n7 = Set Left(Left:double="To Float (Boolean)"(InBool="float < float"(A=@n6, B=-50))) -> (Output_Get:double)
   > n8 = Set Right(Right:double="To Float (Boolean)"(InBool="float > float"(A=@n6, B=50))) -> (Output_Get:double)
   n9 = GetCurveValue(CurveName=crouching_side)
   > Set crouching side(crouching side:double=float * float(A:bool="Not Equal (Float)"(A=@n7, B=@n8), B=float - float(A=1, B=@n9)) {is=title K2Node_PromotableOperator}) -> (Output_Get:double)
   > n10 = Cast To FortPlayerPawn(Object=TryGetPawnOwner())
   n11 = GetSkeletalMeshForPartType(self=@n10, PartType=Backpack)
   > Set backpack subtlesilly(backpack subtlesilly:bool=MatchesWildcard(SourceString=GetDisplayName(Object=@n11), Wildcard=*M_MED_SubtleSilly_Pack*, SearchCase=CaseSensitive)) -> (Output_Get:bool)
   n12 = Get bIsSliding()
   > Set crouching wt sliding(crouching wt sliding:bool=AND Boolean(A=NOT Boolean(A=@n12), B=Get bIsCrouching())) -> (Output_Get:bool)
   n13 = NOT Boolean(A=Get bIsCrouching())
   n14 = Get bIsSwimSprinting() -> (bIsSwimSprinting:bool)
   > Set sprinting modified(sprinting modified:bool=AND Boolean(A=Get bIsSprinting(), B=OR Boolean(A=@n13, B=@n14))) -> (Output_Get:bool)
   n15 = Get bIsInFrontEnd()
   > Set isNotFrontEnd(isNotFrontEnd:bool=NOT Boolean(A=@n15)) -> (Output_Get:bool)
   > n16 = Branch(Condition=Get bIsSkyDiving())
   > Set rt strap fix skydive(rt strap fix skydive:double=0.35) -> (Output_Get:double)
   n17 = Branch(execute=@n16.else, Condition=@n15)
   > Set rt strap fix skydive(rt strap fix skydive=0)
   Set rt strap fix skydive(execute=@n17.else, rt strap fix skydive=1)
   Branch(execute=@n1.then_1, Condition=Get bIsInFrontEnd())
   > Set!(
    execute
    Sim Space Settings_SimSpaceAngularAlpha:float=0.15
    Sim Space Settings_MaxAngularVelocity:float=7
    Sim Space Settings_MaxAngularAcceleration:float=25
    Sim Space Settings_ExternalAngularVelocity:Vector="0,0,0"
   ) -> (then, Output_Get:Anim Phys Sim Space Settings Structure)
  ]}
]
variables [
 backpack subtlesilly {type=boolean}
 crouching side {type=number}
 crouching wt sliding {type=boolean}
 GravityOverride {type=Vector}
 isNotFrontEnd {type=boolean}
 Left {type=number}
 Right {type=number}
 rt strap fix skydive {type=number}
 Sim Space Settings {
  type=AnimPhysSimSpaceSettings
  default {maxAngularAcceleration=35, maxAngularVelocity=7, simSpaceAngularAlpha=0.5, externalAngularVelocity=<0 0 0>}
 }
 sprinting modified {type=boolean}
 thigh side {type=number}
 thigh up {type=number}
]
