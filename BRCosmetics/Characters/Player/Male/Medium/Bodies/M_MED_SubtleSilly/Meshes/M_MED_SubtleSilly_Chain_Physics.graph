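# Physics asset for the hanging chain on the SubtleSilly medium male player body cosmetic.
# Two three-segment capsule strands hang from small base bodies, with limited swing and zero twist limits at each joint.
# A constraint with free swing and twist joins the two end segments, closing the strands into a loop.
# Three large capsules on right-side front, side and back skirt bones have no constraints, likely serving as colliders for the chain.
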PhysicsAsset /BRCosmetics/Characters/Player/Male/Medium/Bodies/M_MED_SubtleSilly/Meshes/M_MED_SubtleSilly_Chain_Physics
bodies {
 base_chain_0 [{shapeType=Capsule, length=0.505, radius=0.505, shapeName=base_chain_0_capsule, center=<2 0 0>, rotation=<rot 0 -90 0>}]
 base_chain_3 [{shapeType=Capsule, length=0.505, radius=0.505, shapeName=base_chain_3_capsule, rotation=<rot 0 -90 0>}]
 dyn_chain_0 [{shapeType=Capsule, length=10, radius=2, shapeName=dyn_chain_0_capsule, center=<2 0.2939 0>, rotation=<rot 90 0 0>}]
 dyn_chain_1 [{shapeType=Capsule, length=10, radius=2, shapeName=dyn_chain_1_capsule, center=<2 -0.1884 0>, rotation=<rot 82.7773 179.9999 179.9999>}]
 dyn_chain_2 [
  {shapeType=Capsule, length=6.74, radius=2, shapeName=dyn_chain_2_capsule, center=<2 -0.3027 0>, rotation=<rot 79.1692 0.0001 0.0001>}
  {shapeType=Capsule, length=7.7836, radius=2, shapeName=dyn_chain_2_capsule, center=<6.3042 -0.1916 -2.9571>, rotation=<rot 39.6632 0 -5.2655>}
 ]
 dyn_chain_3 [{shapeType=Capsule, length=10, radius=2, shapeName=dyn_chain_3_capsule, center=<2.4922 -0.1826 0>, rotation=<rot 90 -0.5086 0>}]
 dyn_chain_4 [{shapeType=Capsule, length=10, radius=2, shapeName=dyn_chain_4_capsule, center=<1.9162 -0.1711 0>, rotation=<rot 90 0.9944 0>}]
 dyn_chain_5 [
  {shapeType=Capsule, length=10, radius=2, shapeName=dyn_chain_5_capsule, center=<1.9919 -0.837 0>, rotation=<rot 90 -4.2307 0>}
  {
   shapeType=Capsule
   length=7.5094
   radius=2
   shapeName=dyn_chain_5_capsule
   center=<10.5763 -1.5942 3.2182>
   rotation=<rot 34.4101 -179.8879 -173.5492>
  }
 ]
 dyn_skirt_back_side_r_0 [{shapeType=Capsule, length=43.7291, radius=5.743, center=<12.238 7.4214 -1.7067>, rotation=<rot 54.5101 60.206 67.386>}]
 dyn_skirt_front_r_0 [{shapeType=Capsule, length=44.1428, radius=5.3292, center=<19.1804 5.1125 -4.6609>, rotation=<rot 58.4288 -124.5846 -126.1413>}]
 dyn_skirt_side_r_0 [{shapeType=Capsule, length=43.951, radius=5.5211, center=<15.0579 13.0277 -2.6879>, rotation=<rot 85.8242 129.6365 152.7662>}]
}
constraints [
 {bone1Name=dyn_chain_0, bone2Name=base_chain_0, swing1LimitDegrees=25, swing2LimitDegrees=10, twistLimitDegrees=0}
 {bone1Name=dyn_chain_1, bone2Name=dyn_chain_0, swing1LimitDegrees=25, swing2LimitDegrees=10, twistLimitDegrees=0}
 {bone1Name=dyn_chain_2, bone2Name=dyn_chain_1, swing1LimitDegrees=25, swing2LimitDegrees=10, twistLimitDegrees=0}
 {bone1Name=dyn_chain_3, bone2Name=base_chain_3, swing1LimitDegrees=20, swing2LimitDegrees=10, twistLimitDegrees=0}
 {bone1Name=dyn_chain_4, bone2Name=dyn_chain_3, swing1LimitDegrees=20, swing2LimitDegrees=10, twistLimitDegrees=0}
 {bone1Name=dyn_chain_5, bone2Name=dyn_chain_4, swing1LimitDegrees=20, swing2LimitDegrees=10, twistLimitDegrees=0}
 {
  bone1Name=dyn_chain_5
  bone2Name=dyn_chain_2
  swing1LimitDegrees=45
  swing1Motion=Free
  swing2LimitDegrees=45
  swing2Motion=Free
  twistLimitDegrees=45
  twistMotion=Free
 }
]
