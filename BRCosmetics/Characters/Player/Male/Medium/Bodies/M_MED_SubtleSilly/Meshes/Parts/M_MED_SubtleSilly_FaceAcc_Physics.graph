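# Physics asset for the face accessory mesh of the SubtleSilly medium male player body, defining a jointed ponytail chain.
# Five dyn_ponytail bodies link in sequence from base_ponytail_1 through constraints with small swing limits and little or no twist.
# Capsules on the head, neck, clavicles, upper spine and dyn_hat have no constraints, apparently serving as colliders for the ponytail.
# All shapes are capsules, with dyn_ponytail_4 using two.
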PhysicsAsset /BRCosmetics/Characters/Player/Male/Medium/Bodies/M_MED_SubtleSilly/Meshes/Parts/M_MED_SubtleSilly_FaceAcc_Physics
bodies {
 base_ponytail_1 [{shapeType=Capsule, length=0.505, radius=0.505, shapeName=base_ponytail_1_capsule, rotation=<rot 0 -90 0>}]
 clavicle_l [{
   shapeType=Capsule
   length=12.5
   radius=9.6
   shapeName=clavicle_l_capsule
   center=<7.1364 -0.4291 1.527>
   rotation=<rot 78.2886 0.7933 3.5828>
  }]
 clavicle_r [{shapeType=Capsule, length=12.5, radius=9.6, shapeName=clavicle_l_capsule, center=<-7 0 -1.6>, rotation=<rot 90 0 0>}]
 dyn_hat [{shapeType=Capsule, length=27.7519, radius=2.2481, shapeName=head_capsule, center=<-0.1964 -18.7242 3.3124>, rotation=<rot 90 0 0>}]
 dyn_ponytail_1 [{shapeType=Capsule, length=4.3447, radius=1.6553, shapeName=dyn_ponytail_1_capsule, center=<2 0 0>, rotation=<rot 23.2649 0 0>}]
 dyn_ponytail_2 [{shapeType=Capsule, length=4.9875, radius=1.5384, shapeName=dyn_ponytail_2_capsule, center=<2 0 0>, rotation=<rot 62.8515 0 0>}]
 dyn_ponytail_3 [{shapeType=Capsule, length=4.4105, radius=1.5895, shapeName=dyn_ponytail_3_capsule, center=<2 0 0>, rotation=<rot 90 0 0>}]
 dyn_ponytail_4 [
  {shapeType=Capsule, length=4.7998, radius=2.1778, shapeName=dyn_ponytail_4_capsule, center=<3 0 0>, rotation=<rot 88.5084 0 0>}
  {
   shapeType=Capsule
   length=5.4691
   radius=1.525
   shapeName=dyn_ponytail_4_capsule
   center=<4.1761 3.9562 -0.4637>
   rotation=<rot 88.5806 -180 -180>
  }
 ]
 dyn_ponytail_5 [{shapeType=Capsule, length=8.899, radius=2.1853, shapeName=dyn_ponytail_5_capsule, center=<4.6441 -1.0072 0.0187>, rotation=<rot 80.2975 0 0>}]
 head [{shapeType=Capsule, length=10, radius=5, shapeName=head_capsule, center=<0 -3.6888 0>}]
 neck_01 [{shapeType=Capsule, length=11.2835, radius=6.8732, shapeName=neck_01_capsule, rotation=<rot 90 7.3196 0>}]
 spine_04 [{shapeType=Capsule, length=27.6568, radius=11.2234, shapeName=spine_05_capsule, center=<5.7866 -2.8328 0>}]
 spine_05 [{shapeType=Capsule, length=28.0249, radius=10.8554, shapeName=spine_05_capsule, center=<1.2815 0 0>}]
}
constraints [
 {bone1Name=dyn_ponytail_1, bone2Name=base_ponytail_1, swing1LimitDegrees=5, swing2LimitDegrees=10, twistLimitDegrees=0}
 {bone1Name=dyn_ponytail_2, bone2Name=dyn_ponytail_1, swing1LimitDegrees=10, swing2LimitDegrees=15, twistLimitDegrees=1}
 {bone1Name=dyn_ponytail_3, bone2Name=dyn_ponytail_2, swing1LimitDegrees=10, swing2LimitDegrees=15, twistLimitDegrees=1}
 {bone1Name=dyn_ponytail_4, bone2Name=dyn_ponytail_3, swing1LimitDegrees=10, swing2LimitDegrees=15, twistLimitDegrees=1}
 {bone1Name=dyn_ponytail_5, bone2Name=dyn_ponytail_4, swing1LimitDegrees=10, swing2LimitDegrees=15, twistLimitDegrees=1}
]
