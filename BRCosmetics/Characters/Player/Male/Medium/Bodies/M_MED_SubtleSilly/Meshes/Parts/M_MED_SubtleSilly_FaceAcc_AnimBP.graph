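# Post-process style Anim Blueprint for the SubtleSilly player face accessory, animating hair, ponytail and hat.
# Copied parent and body poses feed a Control Rig, then a ponytail bone offset, front-hair AnimDynamics, neck-relative RigidBody and spine_05 ponytail constraints.
# The DefaultSlot overrides hat bones, weighted by the front-end idle curve.
# The event graph computes gravity override, a backpack-dependent spine-attach flag and front-end simulation settings.
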AnimBlueprint /BRCosmetics/Characters/Player/Male/Medium/Bodies/M_MED_SubtleSilly/Meshes/Parts/M_MED_SubtleSilly_FaceAcc_AnimBP
target_skeleton=&/M_MED_SubtleSilly_FaceAcc_Skeleton
graphs [
 AnimGraph {nodes [
   Output Pose(Result=Layered blend per bone(
     BasePose=Use cached pose 'dynamicsCache'()
     BlendPoses_0=Slot 'DefaultSlot'()
     BlendWeights_0=Get Custom Idle Is Playing Float() -> (Custom Idle Is Playing Float:double)
    ) {layerSetup [{branchFilters [{boneName=base_hat}, {boneName=dyn_hat}]}]}) {layerGroup=None}
   n1 = Get Body_Mesh()
   n2 = Control Rig!(
    Source=Layered blend per bone(
     BasePose=Copy Pose From Mesh() {bUseAttachedParent=true}
     BlendPoses_0=Copy Pose From Mesh(SourceMeshComponent=@n1) {bUseAttachedParent=true}
    ) {layerSetup [{branchFilters [{boneName=clavicle_l}, {boneName=clavicle_r}]}]}
    Alpha
    isTargeting:bool=Get bIsTargeting()
    isCrouching:bool=Get bIsCrouching()
   ) -> (Pose) {controlRigAssetReference {controlRigAsset=&/M_MED_SubtleSilly_FaceAcc_CtrlRig}}
   Save cached pose 'controlRigCache'(Pose=@n2) {cacheName=controlRigCache}
   n3 = Anim Dynamics - Bone: dyn_hair_front!(
    ComponentPose=Local To Component(LocalPose=Use cached pose 'transformModifyCache'())
    Alpha=1
    SimSpaceSettings:Anim Phys Sim Space Settings Structure=(SimSpaceAngularAlpha=0,MaxAngularVelocity=10000,MaxAngularAcceleration=10000,ExternalAngularVelocity=(X=0,Y=0,Z=0))
   ) -> (Pose) {
    angularDampingOverride=1
    angularSpringConstant=20000
    bAngularSpring=true
    bOverrideAngularDamping=true
    bUseGravityOverride=true
    gravityScale=0.35
    simulationSpace=RootRelative
    boundBone {boneName=dyn_hair_front}
    componentAppliedLinearAccClamp=<3000 3000 3000>
    componentLinearAccScale=<0.3 0.3 0.3>
    componentLinearVelScale=<0.3 0.3 0.3>
    gravityOverride {z=-980}
    physicsBodyDefinitions [{
      boundBone {boneName=dyn_hair_front}
      boxExtents=<45 45 45>
      constraintSetup {angularLimitsMax=<10 5 20>, angularLimitsMin=<-10 -5 -15>, angularTarget=<1 0 0>}
      localJointOffset=<8 0 0>
     }]
    simSpaceSettings {maxAngularAcceleration=35, maxAngularVelocity=7, simSpaceAngularAlpha=1}
   }
   n4 = Anim Dynamics - Bone: dyn_hair_front_lf_3!(
    ComponentPose=@n3
    Alpha=1
    SimSpaceSettings:Anim Phys Sim Space Settings Structure=(SimSpaceAngularAlpha=1,MaxAngularVelocity=7,MaxAngularAcceleration=35,ExternalAngularVelocity=(X=0,Y=0,Z=0))
   ) -> (Pose) {
    angularDampingOverride=1
    angularSpringConstant=20000
    bAngularSpring=true
    bOverrideAngularDamping=true
    bUseGravityOverride=true
    gravityScale=0.35
    simulationSpace=RootRelative
    boundBone {boneName=dyn_hair_front_lf_3}
    componentAppliedLinearAccClamp=<3000 3000 3000>
    componentLinearAccScale=<0.3 0.3 0.3>
    componentLinearVelScale=<0.3 0.3 0.3>
    gravityOverride {z=-980}
    physicsBodyDefinitions [{
      boundBone {boneName=dyn_hair_front_lf_3}
      boxExtents=<45 45 45>
      constraintSetup {angularLimitsMax=<10 3 5>, angularLimitsMin=<-10 -3 -5>, angularTarget=<1 0 0>}
      localJointOffset=<8 0 0>
     }]
    simSpaceSettings {maxAngularAcceleration=35, maxAngularVelocity=7, simSpaceAngularAlpha=1}
   }
   n5 = Anim Dynamics - Bone: dyn_hair_front_rt_1!(
    ComponentPose=@n4
    Alpha=1
    SimSpaceSettings:Anim Phys Sim Space Settings Structure=(SimSpaceAngularAlpha=1,MaxAngularVelocity=7,MaxAngularAcceleration=35,ExternalAngularVelocity=(X=0,Y=0,Z=0))
   ) -> (Pose) {
    angularDampingOverride=1
    angularSpringConstant=20000
    bAngularSpring=true
    bOverrideAngularDamping=true
    bUseGravityOverride=true
    gravityScale=0.35
    simulationSpace=RootRelative
    boundBone {boneName=dyn_hair_front_rt_1}
    componentAppliedLinearAccClamp=<3000 3000 3000>
    componentLinearAccScale=<0.3 0.3 0.3>
    componentLinearVelScale=<0.3 0.3 0.3>
    gravityOverride {z=-980}
    physicsBodyDefinitions [{
      boundBone {boneName=dyn_hair_front_rt_1}
      boxExtents=<45 45 45>
      constraintSetup {angularLimitsMax=<10 5 20>, angularLimitsMin=<-10 -5 -15>, angularTarget=<1 0 0>}
      localJointOffset=<8 0 0>
     }]
    simSpaceSettings {maxAngularAcceleration=35, maxAngularVelocity=7, simSpaceAngularAlpha=1}
   }
   n6 = Anim Dynamics - Bone: dyn_hair_front_rt_2!(
    ComponentPose=@n5
    Alpha=1
    SimSpaceSettings:Anim Phys Sim Space Settings Structure=(SimSpaceAngularAlpha=1,MaxAngularVelocity=7,MaxAngularAcceleration=35,ExternalAngularVelocity=(X=0,Y=0,Z=0))
   ) -> (Pose) {
    angularDampingOverride=1
    angularSpringConstant=20000
    bAngularSpring=true
    bOverrideAngularDamping=true
    bUseGravityOverride=true
    gravityScale=0.35
    simulationSpace=RootRelative
    boundBone {boneName=dyn_hair_front_rt_2}
    componentAppliedLinearAccClamp=<3000 3000 3000>
    componentLinearAccScale=<0.3 0.3 0.3>
    componentLinearVelScale=<0.3 0.3 0.3>
    gravityOverride {z=-980}
    physicsBodyDefinitions [{
      boundBone {boneName=dyn_hair_front_rt_2}
      boxExtents=<45 45 45>
      constraintSetup {angularLimitsMax=<10 5 20>, angularLimitsMin=<-10 -5 -15>, angularTarget=<1 0 0>}
      localJointOffset=<8 0 0>
     }]
    simSpaceSettings {maxAngularAcceleration=35, maxAngularVelocity=7, simSpaceAngularAlpha=1}
   }
   n7 = Anim Dynamics - Bone: dyn_hair_front_lf_1!(
    ComponentPose=@n6
    Alpha=1
    SimSpaceSettings:Anim Phys Sim Space Settings Structure=(SimSpaceAngularAlpha=1,MaxAngularVelocity=7,MaxAngularAcceleration=35,ExternalAngularVelocity=(X=0,Y=0,Z=0))
   ) -> (Pose) {
    angularDampingOverride=1
    angularSpringConstant=20000
    bAngularSpring=true
    bOverrideAngularDamping=true
    bUseGravityOverride=true
    gravityScale=0.35
    simulationSpace=RootRelative
    boundBone {boneName=dyn_hair_front_lf_1}
    componentAppliedLinearAccClamp=<3000 3000 3000>
    componentLinearAccScale=<0.3 0.3 0.3>
    componentLinearVelScale=<0.3 0.3 0.3>
    gravityOverride {z=-980}
    physicsBodyDefinitions [{
      boundBone {boneName=dyn_hair_front_lf_1}
      boxExtents=<45 45 45>
      constraintSetup {angularLimitsMax=<10 3 5>, angularLimitsMin=<-10 -3 -5>, angularTarget=<1 0 0>}
      localJointOffset=<8 0 0>
     }]
    simSpaceSettings {maxAngularAcceleration=35, maxAngularVelocity=7, simSpaceAngularAlpha=1}
   }
   n8 = Anim Dynamics - Bone: dyn_hair_front_lf_2!(
    ComponentPose=@n7
    Alpha=1
    SimSpaceSettings:Anim Phys Sim Space Settings Structure=(SimSpaceAngularAlpha=1,MaxAngularVelocity=7,MaxAngularAcceleration=35,ExternalAngularVelocity=(X=0,Y=0,Z=0))
   ) -> (Pose) {
    angularDampingOverride=1
    angularSpringConstant=20000
    bAngularSpring=true
    bOverrideAngularDamping=true
    bUseGravityOverride=true
    gravityScale=0.35
    simulationSpace=RootRelative
    boundBone {boneName=dyn_hair_front_lf_2}
    componentAppliedLinearAccClamp=<3000 3000 3000>
    componentLinearAccScale=<0.3 0.3 0.3>
    componentLinearVelScale=<0.3 0.3 0.3>
    gravityOverride {z=-980}
    physicsBodyDefinitions [{
      boundBone {boneName=dyn_hair_front_lf_2}
      boxExtents=<45 45 45>
      constraintSetup {angularLimitsMax=<10 3 5>, angularLimitsMin=<-10 -3 -5>, angularTarget=<1 0 0>}
      localJointOffset=<8 0 0>
     }]
    simSpaceSettings {maxAngularAcceleration=35, maxAngularVelocity=7, simSpaceAngularAlpha=1}
   }
   n9 = RigidBody!(ComponentPose=@n8, Alpha, OverrideWorldGravity:Vector="0,0,0", ExternalForce) -> (Pose) {
    bClampLinearTranslationLimitToRefPose=true
    bOverrideWorldGravity=true
    overridePhysicsAsset=&/M_MED_SubtleSilly_FaceAcc_Physics
    simulationSpace=BaseBoneSpace
    baseBoneRef {boneName=neck_02}
    simSpaceSettings {dampingAlpha=0, maxAngularAcceleration=10, maxAngularVelocity=5, maxLinearAcceleration=1200, maxLinearVelocity=600, worldAlpha=1}
   }
   n10 = "Constraint - dyn_ponytail_1 (P)"!(ComponentPose=@n9, bAlphaBoolEnabled:bool=True, ConstraintWeights_0=1) -> (Pose) {alphaInputType=Bool, boneToModify {boneName=dyn_ponytail_1}, constraintSetup [{targetBone {boneName=spine_05}}], constraintWeights [1]}
   n11 = "Constraint - dyn_ponytail_2 (P)"!(ComponentPose=@n10, bAlphaBoolEnabled:bool=True, ConstraintWeights_0=1) -> (Pose) {alphaInputType=Bool, boneToModify {boneName=dyn_ponytail_2}, constraintSetup [{targetBone {boneName=spine_05}}], constraintWeights [1]}
   n12 = "Constraint - dyn_ponytail_3 (P)"!(ComponentPose=@n11, bAlphaBoolEnabled:bool=True, ConstraintWeights_0=1) -> (Pose) {alphaInputType=Bool, boneToModify {boneName=dyn_ponytail_3}, constraintSetup [{targetBone {boneName=spine_05}}], constraintWeights [1]}
   n13 = "Constraint - dyn_ponytail_4 (P)"!(ComponentPose=@n12, bAlphaBoolEnabled:bool=True, ConstraintWeights_0=1) -> (Pose) {alphaInputType=Bool, boneToModify {boneName=dyn_ponytail_4}, constraintSetup [{targetBone {boneName=spine_05}}], constraintWeights [1]}
   n14 = "Constraint - dyn_ponytail_4 (P)"(ComponentPose=@n13, bAlphaBoolEnabled=True, ConstraintWeights_0=1) {alphaInputType=Bool, boneToModify {boneName=dyn_ponytail_4}, constraintSetup [{targetBone {boneName=spine_05}}], constraintWeights [1]}
   n15 = "Transform (Modify) Bone - Bone: base_ponytail_1"!(ComponentPose=@n14, bAlphaBoolEnabled:bool=True, Rotation="10,10,0") -> (Pose) {
    alphaInputType=Bool
    rotationMode=BMM_Additive
    rotationSpace=BCS_BoneSpace
    boneToModify {boneName=base_ponytail_1}
    rotation {pitch=10, yaw=10}
   }
   Save cached pose 'dynamicsCache'(Pose=Component To Local(ComponentPose=@n15)) {cacheName=dynamicsCache}
   n16 = "Transform (Modify) Bone - Bone: base_ponytail_1"(ComponentPose=Local To Component(LocalPose=Use cached pose 'controlRigCache'()), bAlphaBoolEnabled=True, Rotation="-45,0,0") {alphaInputType=Bool, rotationMode=BMM_Additive, rotationSpace=BCS_BoneSpace, boneToModify {boneName=base_ponytail_1}, rotation {pitch=-45}}
   Save cached pose 'transformModifyCache'(Pose=Component To Local(ComponentPose=@n16)) {cacheName=transformModifyCache}
  ]}
 EventGraph {nodes [
   Event OnCharacterCustomizationLoaded()
   > n1 = Cast To PlayerPawn_Athena(Object=TryGetPawnOwner())
   > Set Body_Mesh(Body_Mesh=GetSkeletalMeshForPartType(self=@n1))
   Event BlueprintUpdateAnimation()
   > n2 = Sequence()
   > n3 = "Gravity Override State Function (Parameter Asset) (Nativized)"(
    CharacterPartAnimInstance=Self-Reference()
    GameTimeInSeconds=GetGameTimeInSeconds()
    Parameters=&/M_MED_SubtleSilly_FaceAcc_Parameters
   )
   > Set GravityOverride(GravityOverride=@n3)
   > n4 = Cast To FortPlayerPawn(Object=TryGetPawnOwner())
   n5 = GetSkeletalMeshForPartType(self=@n4, PartType=Backpack)
   > n6 = Set backpack subtlesilly(backpack subtlesilly:bool=MatchesWildcard(SourceString=GetDisplayName(Object=@n5), Wildcard=*M_MED_SubtleSilly_Pack*, SearchCase=CaseSensitive)) -> (Output_Get:bool)
   > n7 = Set isWearingBackbling(isWearingBackbling=CheckForBackbling())
   > n8 = Branch(Condition=@n7)
   > n9 = Branch(Condition=AND Boolean(A=@n7, B=@n6))
   > Set attatchToSpine(attatchToSpine:bool=false) -> (Output_Get:bool)
   Set attatchToSpine(execute=@n9.else, attatchToSpine=true)
   Set attatchToSpine(execute=@n8.else, attatchToSpine=false)
   Branch(execute=@n2.then_1, Condition=Get bIsInFrontEnd())
   > Set Custom Idle Is Playing Float(Custom Idle Is Playing Float:double=GetCurveValue(self=GetOwnerFrontEndAnimInstance(), CurveName=Frontend_Idle_Bypass)) -> (Output_Get:double)
   > Branch(Condition=Get bIsInFrontEnd())
   > Set!(
    execute
    Sim Space Settings_SimSpaceAngularAlpha:float=0.3
    Sim Space Settings_MaxAngularVelocity:float=7
    Sim Space Settings_MaxAngularAcceleration:float=35
    Sim Space Settings_ExternalAngularVelocity:Vector="0,0,0"
   ) -> (then, Output_Get:Anim Phys Sim Space Settings Structure)
  ]}
]
variables [
 attatchToSpine {type=boolean}
 backpack subtlesilly {type=boolean}
 Body_Mesh {type=SkeletalMeshComponent}
 Custom Idle Is Playing Float {type=number}
 GravityOverride {type=Vector}
 isWearingBackbling {type=boolean}
 Sim Space Settings {
  type=AnimPhysSimSpaceSettings
  default {maxAngularAcceleration=35, maxAngularVelocity=7, simSpaceAngularAlpha=1, externalAngularVelocity=<0 0 0>}
 }
]
